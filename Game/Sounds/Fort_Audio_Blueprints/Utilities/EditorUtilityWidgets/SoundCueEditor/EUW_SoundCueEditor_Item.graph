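# Editor utility list row for the EUW_SoundCueEditor tool, showing one sound node's class name and details.
# It stores its list item object, then OnPostInitialize saves the parent editor and calls its GetSoundNodeDetails for text and depth.
# Depth sets the border's left padding, indenting rows like a tree.
# Single and double mouse clicks both route to OnClicked with a flag, but OnClicked and OnSelectedChanged are empty here.
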WidgetBlueprint /Game/Sounds/Fort_Audio_Blueprints/Utilities/EditorUtilityWidgets/SoundCueEditor/EUW_SoundCueEditor_Item
graphs [
 EventGraph {nodes [
   n1 = Event OnListItemObjectSet()
   > Set EntryObject(EntryObject:Object Reference=@n1.ListItemObject) -> (Output_Get:Object Reference)
  ]}
 OnClicked {nodes [OnClicked!() -> (then, MouseEvent:"Pointer Event Structure (by ref)", bDoubleClick:bool=~) {is=title K2Node_FunctionEntry}]}
 OnMouseClicked {nodes [
   n1 = OnMouseClicked!() -> (then, MyGeometry:Geometry Structure, MouseEvent:"Pointer Event Structure (by ref)") {is=title K2Node_FunctionEntry}
   > OnClicked!(execute, self, MouseEvent:"Pointer Event Structure (by ref)"=@n1.MouseEvent, bDoubleClick:bool=false) -> (then) {is=title K2Node_CallFunction}
   > Return Node!(execute, ReturnValue:Event Reply Structure=Make Event Reply!() -> (EventReply:Event Reply Structure) {is=title K2Node_MakeStruct})
  ]}
 OnMouseDoubleClick {nodes [
   n1 = OnMouseDoubleClick!() -> (then, MyGeometry:Geometry Structure, MouseEvent:"Pointer Event Structure (by ref)") {is=title K2Node_FunctionEntry}
   > OnClicked(MouseEvent=@n1.MouseEvent, bDoubleClick=true)
   > Return Node(ReturnValue=Make Event Reply())
  ]}
 OnPostInitialize {nodes [
   n1 = OnPostInitialize!() -> (then, Parent Widget:EUW Sound Cue Editor Object Reference) {is=title K2Node_FunctionEntry}
   > Set ParentWidget(ParentWidget:EUW Sound Cue Editor Object Reference=@n1) -> (Output_Get:EUW Sound Cue Editor Object Reference)
   n2 = Get EntryObject() -> (EntryObject:Object Reference)
   n3 = GetSoundNodeDetails!(
    self:EUW Sound Cue Editor Object Reference=Get ParentWidget() -> (ParentWidget:EUW Sound Cue Editor Object Reference)
    Object:Object Reference=@n2
   ) -> (Details:Text, Depth:int=0)
   > SetPadding!(
    execute
    self:Border Object Reference=Get HighlightBorder() -> (HighlightBorder:Border Object Reference)
    InPadding_Left:float=float * float(A="To Float (Integer)"(InInt=@n3.Depth), B=10) {is=title K2Node_PromotableOperator}
    InPadding_Top:float=0
    InPadding_Right:float=0
    InPadding_Bottom:float=0
   ) -> (then)
   n4 = Get Class(Object=@n2) -> (ReturnValue:Object Class Reference)
   > "SetText (Text)"(
    self=Get SoundNodeName() -> (SoundNodeName:Text Block Object Reference)
    InText="To Text (String)"(InString=Get Class Display Name(Class=@n4))
   )
   > "SetText (Text)"(self=Get SoundNodeDetails() -> (SoundNodeDetails:Text Block Object Reference), InText=@n3)
  ]}
 OnSelectedChanged {nodes [OnSelectedChanged!() -> (then, bSelected:bool=~) {is=title K2Node_FunctionEntry}]}
]
variables [
 EntryObject {type=Object, default=null}
 ParentWidget {type=EUW_SoundCueEditor_C, default=null}
]
widgets {
 info {inheritedWidgetCount=0, parentClass=&EditorUtilityWidget, rootWidgetClass=&CanvasPanel, widgetCount=6}
 widgets [
  CanvasPanel_178 {widgetClassPath=&CanvasPanel, widget=CanvasPanel_178}
  HighlightBorder {
   widgetClassPath=&Border
   bIsVariable=true
   parent=CanvasPanel_178
   widget=HighlightBorder
   properties {
    background {
     bIsDynamicallyLoaded=false
     drawAs=Box
     imageType=NoImage
     mirroring=NoMirror
     resourceName=None
     resourceObject=null
     tiling=NoTile
     imageSize=<32 32>
     margin=<ltrb 0 0 0 0>
     outlineSettings {
      bUseBrushTransparency=false
      roundingType=HalfHeightRadius
      width=0
      color {colorUseRule=UseColor_Specified, specifiedColor=<rgba 0 0 0 0>}
      cornerRadii=<0 0 0 0>
     }
     tintColor {colorUseRule=UseColor_Specified, specifiedColor=<rgba 1 1 1 1>}
     uVRegion {bIsValid=false, max=<0 0>, min=<0 0>}
    }
    brushColor=<rgba 1 1 1 0>
    padding=<ltrb 5 5 5 5>
   }
   slot {
    class=CanvasPanelSlot
    properties {bAutoSize=true, layoutData {alignment=<0 0>, anchors {maximum=<1 1>, minimum=<0 0>}, offsets=<ltrb 0 0 0 0>}}
   }
  }
  VerticalBox_142 {
   widgetClassPath=&VerticalBox
   parent=HighlightBorder
   widget=VerticalBox_142
   slot {class=BorderSlot, properties {padding=<ltrb 5 5 5 5>}}
  }
  HorizontalBox_82 {widgetClassPath=&HorizontalBox, parent=VerticalBox_142, widget=HorizontalBox_82, slot {class=VerticalBoxSlot}}
  SoundNodeName {
   widgetClassPath=&UEFN_TextBlock
   bIsVariable=true
   parent=HorizontalBox_82
   widget=SoundNodeName
   properties {
    text=SoundNode
    font {
     bForceMonospaced=false
     bMaterialIsStencil=false
     fontMaterial=null
     fontObject=&NotoSans
     letterSpacing=5
     monospacedWidth=1
     size=10
     skewAmount=0
     typefaceFontName=NotoSans-Regular
     outlineSettings {
      bApplyOutlineToDropShadows=false
      bMiteredCorners=false
      bSeparateFillAlpha=false
      outlineMaterial=null
      outlineSize=0
      outlineColor=<rgba 0 0 0 1>
     }
    }
   }
   slot {class=HorizontalBoxSlot, properties {horizontalAlignment=HAlign_Left, padding=<ltrb 0 0 25 0>}}
  }
  SoundNodeDetails {
   widgetClassPath=&UEFN_TextBlock
   bIsVariable=true
   parent=HorizontalBox_82
   widget=SoundNodeDetails
   properties {
    text=Details
    colorAndOpacity {colorUseRule=UseColor_Specified, specifiedColor=<rgba 0.4062 0.4062 0.4062 1>}
    font {
     bForceMonospaced=false
     bMaterialIsStencil=false
     fontMaterial=null
     fontObject=&NotoSans
     letterSpacing=5
     monospacedWidth=1
     size=10
     skewAmount=0
     typefaceFontName=NotoSans-Regular
     outlineSettings {
      bApplyOutlineToDropShadows=false
      bMiteredCorners=false
      bSeparateFillAlpha=false
      outlineMaterial=null
      outlineSize=0
      outlineColor=<rgba 0 0 0 1>
     }
    }
   }
   slot {class=HorizontalBoxSlot, properties {horizontalAlignment=HAlign_Left}}
  }
 ]
}
