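# SoundCueEditor_Test_Cue is a sound cue in the SoundCueEditor editor utility widget folder, apparently a test fixture for that tool.
# Two wave players with no sound wave assigned feed a mixer at equal input volume.
# The mixer feeds a root continuous modulator, whose pitch and volume are bound to placeholder parameters named Foo and Bar.
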SoundCue /Game/Sounds/Fort_Audio_Blueprints/Utilities/EditorUtilityWidgets/SoundCueEditor/SoundCueEditor_Test_Cue
root=ModulatorContinuous_0
nodes [
 {id=Mixer_0, children [WavePlayer_0,WavePlayer_1], properties {inputVolume [1,1]}}
 {
  id=ModulatorContinuous_0
  children [Mixer_0]
  properties {pitchModulationParams {parameterName=Foo}, volumeModulationParams {parameterName=Bar}}
 }
 {id=WavePlayer_0}
 {id=WavePlayer_1}
]
properties {maxDistance=2097152}
