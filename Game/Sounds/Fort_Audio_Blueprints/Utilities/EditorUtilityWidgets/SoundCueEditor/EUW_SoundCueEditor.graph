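# Editor Utility Widget that shows the node graph of a picked Sound Cue as a tree list in the editor.
# On construct or picker change, it recursively reads ChildNodes through GetEditorProperty, filling the TreeView and a node-to-depth map.
# Item widgets receive this widget on initialization; GetSoundNodeDetails supplies depth plus text for wave player, branch and continuous modulator nodes.
# Also holds asset-registry helpers and Sound Class viewer functions, including a child getter handling only SoundClasses.
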EditorUtilityWidgetBlueprint /Game/Sounds/Fort_Audio_Blueprints/Utilities/EditorUtilityWidgets/SoundCueEditor/EUW_SoundCueEditor
graphs [
 EventGraph {nodes [
   "On Property Changed (SoundCuePicker)"!() -> (OutputDelegate, then, PropertyName:Name=None) {is=title K2Node_ComponentBoundEvent}
   > SCGraph_GenerateTree()
   n1 = "On Entry Initialized (SoundCueGraphTree)"!() -> (OutputDelegate, then, Item:Object Reference, Widget:User Widget Object Reference) {is=title K2Node_ComponentBoundEvent}
   > n2 = Cast To EUW_SoundCueEditor_Item!(execute, Object=@n1.Widget) -> (then, CastFailed, AsEUW Sound Cue Editor Item:EUW Sound Cue Editor Item Object Reference) {is=title K2Node_DynamicCast}
   > OnPostInitialize!(execute, self:EUW Sound Cue Editor Item Object Reference=@n2, Parent Widget:EUW Sound Cue Editor Object Reference=Self-Reference()) -> (then)
   Event Destruct()
   > ClearListItems(self=Get SoundCueGraphTree() -> (SoundCueGraphTree:Tree View Object Reference))
   Event Construct()
   > SCGraph_GenerateTree()
   > SetObject(self=Get SoundCuePicker() -> (SoundCuePicker:Single Property View Object Reference), NewObject=Self-Reference())
  ]}
 GetAssetDependencies {nodes [
   n1 = Inputs!() -> (execute, Object:Object Reference)
   n2 = Get Object Path String(Object=@n1)
   > n3 = Get Dependencies!(
    execute
    self:Asset Registry Interface=GetAssetRegistry()
    PackageName:Name=String To Name(InString=Split(SourceString=@n2, InStr=., SearchDir=FromEnd))
    DependencyOptions:"Asset Registry Dependency Options Structure (by ref)"=Make Asset Registry Dependency Options(bIncludeSoftPackageReferences=True)
   ) -> (then, OutDependencies:Array of Names, ReturnValue:bool) {is=title K2Node_CallFunction}
   > n4 = For Each Loop(Array:Array of Names=@n3) -> (Array Element:Name)
   > n5 = GetAssetsByPackageName!(
    execute
    self:Asset Registry Interface=GetAssetRegistry()
    PackageName:Name=@n4
    bIncludeOnlyOnDiskAssets:bool=false
    bSkipARFilteredAssets:bool=true
   ) -> (then, OutAssetData:Array of Asset Data Structures, ReturnValue:bool)
   > Branch(Condition="integer > integer"(A=Length(TargetArray:Array of Asset Data Structures=@n5), B=0) {is=title K2Node_CallFunction})
   n6 = "Get (a copy)"(Array:Array of Asset Data Structures=@n5) -> (Output:Asset Data Structure)
   > n7 = LoadAsset_Blocking(Asset=To Soft Object Reference(SoftObjectPath=ToSoftObjectPath!(InAssetData:"Asset Data Structure (by ref)"=@n6) -> (ReturnValue:Soft Object Path Structure)))
   > Outputs!(Iteration:Exec, Dependency:Object Reference=@n7, Completed:Exec=@n4.Completed)
  ]}
 GetAssetReferencers {nodes [
   n1 = Inputs!() -> (execute, Object, Class Filter:Object Class Reference)
   n2 = Get Object Path String(Object=@n1)
   > n3 = Get Referencers!(
    execute
    self:Asset Registry Interface=GetAssetRegistry()
    PackageName:Name=String To Name(InString=Split(SourceString=@n2, InStr=., SearchDir=FromEnd))
    ReferenceOptions:"Asset Registry Dependency Options Structure (by ref)"=Make Asset Registry Dependency Options(bIncludeSoftPackageReferences=True)
   ) -> (then, OutReferencers:Array of Names, ReturnValue:bool) {is=title K2Node_CallFunction}
   > n4 = For Each Loop(Array=@n3)
   > n5 = GetAssetsByPackageName(self=GetAssetRegistry(), PackageName=@n4, bIncludeOnlyOnDiskAssets=true, bSkipARFilteredAssets=true)
   > Branch(Condition="integer > integer"(A=Length(TargetArray=@n5), B=0))
   n6 = IsValidClass!(Class:Object Class Reference=@n1.Class Filter) -> (ReturnValue:bool)
   n7 = Get Class Display Name(Class=@n1.Class Filter)
   n8 = "Get (a copy)"(Array=@n5)
   n9 = Split(
    SourceString=BreakTopLevelAssetPath!(TopLevelAssetPath:"Top Level Asset Path Structure (by ref)"=Break Asset Data!(AssetData=@n8) -> (AssetClassPath)) -> (PathString:String)
    InStr=.
    SearchDir=FromEnd
   )
   > Branch(Condition=OR Boolean(A="Equal Exactly (String)"(A=@n9.RightS, B=@n7), B=NOT Boolean(A=@n6)))
   n10 = "Get (a copy)"(Array=@n5)
   > n11 = LoadAsset_Blocking(Asset=To Soft Object Reference(SoftObjectPath=ToSoftObjectPath(InAssetData=@n10)))
   > Outputs!(Iteration, Referencer:Object Reference=@n11, Completed=@n4.Completed)
  ]}
 GetCurrentWorld {nodes [
   GetCurrentWorld() {is=title K2Node_FunctionEntry}
   > n1 = GetPIEWorlds!(execute, bIncludeDedicatedServer:bool=false) -> (then, ReturnValue:Array of World Object References)
   > n2 = Branch(Condition=Is Valid Index!(TargetArray:Array of World Object References=@n1, IndexToTest:int) -> (ReturnValue:bool))
   > Return Node!(execute, ReturnValue:World Object Reference="Get (a copy)"(Array:Array of World Object References=@n1) -> (Output:World Object Reference))
   n3 = GetEditorWorld(execute=@n2.else, self=Get UnrealEditorSubsystem())
   > Is Valid(InputObject=@n3)
   > Return Node(ReturnValue=@n3)
  ]}
 GetSoundNodeDetails {nodes [
   n1 = GetSoundNodeDetails!() -> (then, Object:Object Reference) {is=title K2Node_FunctionEntry}
   > Set LocalDepth(LocalDepth:int=Find(
     TargetMap:Map of Object References to Integers=Get DepthMap() -> (DepthMap:Map of Object References to Integers)
     Key:"Object Reference (by ref)"=@n1
    ) -> (Value:int)) -> (Output_Get:int=0)
   > n2 = Sequence()
   > n3 = Switch on String!(execute:Exec, Selection:String=Get Class Display Name(Class=Get Class(Object=@n1) -> (ReturnValue:Object Class Reference))) -> (Default:Exec, SoundNodeModulatorContinuous:Exec, SoundNodeWavePlayer:Exec, SoundNodeBranch:Exec) {pinNames [SoundNodeModulatorContinuous, SoundNodeWavePlayer, SoundNodeBranch]}
   K2Node_Composite_2 = Format Branch!(execute=@n3.SoundNodeBranch) -> (then:Exec, ReturnValue:Text) {is=title K2Node_Composite}
   > Return Node!(execute, Details:Text=@K2Node_Composite_2, Depth:int=Get LocalDepth() -> (LocalDepth:int=0))
   K2Node_Composite_0 = Format Continuous Modulator!(execute=@n3.SoundNodeModulatorContinuous) -> (then:Exec, Result:Text) {is=title K2Node_Composite}
   > Return Node(Details=@K2Node_Composite_0, Depth=Get LocalDepth() -> (LocalDepth=0))
   K2Node_Composite_1 = Format Wave Player!(execute=@n3.SoundNodeWavePlayer) -> (then:Exec, ReturnValue:Text) {is=title K2Node_Composite}
   > Return Node(Details=@K2Node_Composite_1, Depth=Get LocalDepth() -> (LocalDepth=0))
   Return Node(execute=@n2.then_1, Depth=Get LocalDepth() -> (LocalDepth=0))
  ]}
 Format Continuous Modulator {
  path=GetSoundNodeDetails.K2Node_Composite_0.Format Continuous Modulator
  nodes [
   Inputs!() -> (execute)
   n1 = Get Object() -> (Object:Object Reference)
   > n2 = GetEditorProperty!(execute:Exec, Object:Object Reference=@n1, PropertyName:Name=VolumeModulationParams) -> (then:Exec, PropertyValue:"Modulator Continuous Params Structure (by ref)", ReturnValue:bool)
   > n3 = GetEditorProperty(Object=@n1, PropertyName=PitchModulationParams)
   n4 = Break Modulator Continuous Params!(ModulatorContinuousParams:"Modulator Continuous Params Structure (by ref)"=@n3) -> (ParameterName:Name) {is=title K2Node_BreakStruct}
   n5 = Format Text!(Format="LOCTEXT(\"Pitch {param}\")", param:Text="To Text (Name)"!(InName:Name=@n4) -> (ReturnValue:Text)) -> (Result)
   n6 = "Not Equal (Name)"(A=@n4) -> (ReturnValue=~)
   n7 = Break Modulator Continuous Params(ModulatorContinuousParams=@n2)
   n8 = "Not Equal (Name)"(A=@n7) -> (ReturnValue=~)
   n9 = AND Boolean(A=@n8, B=@n6)
   n10 = Format Text(Format="LOCTEXT(\"Volume {param}\")", param="To Text (Name)"(InName=@n7))
   > Outputs!(
    then:Exec
    Result:Text=Format Text!(
     Format="LOCTEXT(\"({Volume}{Divider}{Pitch})\")"
     Volume:Text=Select(Option 0:Text, Option 1:Text=@n10, Index=@n8) -> (ReturnValue:Text)
     Divider:Text=Select(Option 1="LOCTEXT(\", \")", Index=@n9)
     Pitch:Text=Select(Option 1=@n5, Index=@n6)
    ) -> (Result)
   )
  ]
 }
 Format Wave Player {
  path=GetSoundNodeDetails.K2Node_Composite_1.Format Wave Player
  nodes [
   Inputs()
   > n1 = GetEditorProperty(Object=Get Object(), PropertyName=SoundWaveAssetPtr) -> (PropertyValue:Soft Object Reference)
   > Outputs!(then, ReturnValue:Text="To Text (String)"(InString=GetSoftPtrAssetName!(Asset:Soft Object Reference=@n1) -> (ReturnValue:String)))
  ]
 }
 Format Branch {
  path=GetSoundNodeDetails.K2Node_Composite_2.Format Branch
  nodes [
   Inputs()
   > n1 = GetEditorProperty(Object=Get Object(), PropertyName=BoolParameterName) -> (PropertyValue:Name)
   > Outputs(ReturnValue="To Text (Name)"(InName=@n1))
  ]
 }
 OnBP_OnGetItemChildren {nodes [
   n1 = OnBP_OnGetItemChildren!() -> (then, Item:Object Reference) {is=title K2Node_FunctionEntry}
   n2 = Cast To SoundClass!(execute, Object=@n1) -> (then, CastFailed, AsSound Class:Sound Class Object Reference) {is=title K2Node_DynamicCast}
   > n3 = GetEditorProperty(Object=@n2, PropertyName=ChildClasses) -> (PropertyValue:Array of Object References)
   > Return Node!(execute, Children:Array of Object References=@n3)
   Return Node(execute=[@n2.CastFailed, @n1.then])
  ]}
 SCGraph_GenerateTree {nodes [
   SCGraph_GenerateTree() {is=title K2Node_FunctionEntry}
   > ClearListItems(self=Get SoundCueGraphTree())
   > Clear(TargetArray:Array of Object References=Get SoundCueNodes() -> (SoundCueNodes:Array of Object References))
   > SCGraph_GetChildrenRecursive!(
    execute
    self
    Node:Object Reference=Get FirstNode!(self:Sound Cue Object Reference=Get SelectedCue() -> (SelectedCue:Sound Cue Object Reference)) -> (FirstNode:Sound Node Object Reference)
    NodeParent:Object Reference
    Depth:int=0
   ) -> (then)
  ]}
 SCGraph_GetChildrenRecursive {nodes [
   n1 = SCGraph_GetChildrenRecursive!() -> (then, Node:Object Reference, NodeParent:Object Reference, Depth:int) {is=title K2Node_FunctionEntry}
   > Is Valid(InputObject=@n1)
   > LoadClassAsset_Blocking(AssetClass=Make Soft Reference!(Input:Object Class Reference=Get Class(Object=@n1)) -> (Output:Object Soft Class Reference))
   > AddItem!(execute, self:List View Object Reference=Get SoundCueGraphTree(), Item:Object Reference=@n1) -> (then)
   > Add Unique(TargetArray:Array of Object References=Get SoundCueNodes(), NewItem:"Object Reference (by ref)"=@n1)
   > Add!(
    execute
    TargetMap:Map of Object References to Integers=Get DepthMap()
    Key:"Object Reference (by ref)"=@n1
    Value:"Integer (by ref)"=@n1.Depth
   ) -> (then) {is=title K2Node_CallFunction}
   > n2 = GetEditorProperty(Object=@n1, PropertyName=ChildNodes)
   > n3 = For Each Loop(Array:Array of Object References=@n2) -> (Array Element:Object Reference)
   > SCGraph_GetChildrenRecursive(Node=@n3, NodeParent=@n1.NodeParent, Depth=int + int(A=@n1.Depth)) {is=title K2Node_CallFunction}
  ]}
 SelectListViewItem {nodes [
   n1 = SelectListViewItem!() -> (then, List View:List View Object Reference, Widget:Object Reference) {is=title K2Node_FunctionEntry}
   > n2 = For Each Loop(Array:Array of User Widget Object References=GetDisplayedEntryWidgets!(self:List View Base Object Reference=@n1) -> (ReturnValue:Array of User Widget Object References)) -> (Array Element:User Widget Object Reference)
   OnSelectedChanged!(
    execute=@n2.Completed
    self:EUW Sound Class Viewer Item Parent Object Reference=Cast To EUW_SoundClassViewer_ItemParent!(Object=@n1.Widget) -> (AsEUW Sound Class Viewer Item Parent:EUW Sound Class Viewer Item Parent Object Reference, bSuccess:bool=~) {is=title K2Node_DynamicCast}
    bSelected:bool=true
   ) -> (then)
   OnSelectedChanged(execute=@n2.LoopBody, self=Cast To EUW_SoundClassViewer_ItemParent(Object=@n2) -> (bSuccess=~), bSelected=false)
  ]}
]
variables [
 Asset Paths {type="array<string>", default []}
 DepthMap {default {}}
 EditChildrenCollection {type=string, default=SoundClassViewer_Workspace}
 SelectedClassWidget {type=EUW_SoundClassViewer_Item_SoundClass_C, default=null}
 SelectedCue {
  type=SoundCue
  default=&/SoundCueEditor_Test_Cue
  class=~
  properties=~
 }
 SoundClassAssetActions {type=SoundClassAssetActions_C, default=null}
 SoundCueNodes {type="array<Object>", default []}
 SoundMixReferencers {type="array<SoundMix>", default []}
]
widgets {
 info {rootWidgetClass=&CanvasPanel, widgetCount=10}
 widgets [
  CanvasPanel_0 {widgetClassPath=&CanvasPanel, widget=CanvasPanel_0}
  SizeBox_0 {
   widgetClassPath=&SizeBox
   parent=CanvasPanel_0
   widget=SizeBox_0
   slot {class=CanvasPanelSlot, properties {layoutData {alignment=<0 0>, anchors {maximum=<1 1>, minimum=<0 0>}, offsets=<ltrb 0 0 0 0>}}}
  }
  HorizontalBox_180 {widgetClassPath=&HorizontalBox, parent=SizeBox_0, widget=HorizontalBox_180, slot {class=SizeBoxSlot}}
  VerticalBox_38 {
   widgetClassPath=&VerticalBox
   parent=HorizontalBox_180
   widget=VerticalBox_38
   slot {class=HorizontalBoxSlot, properties {size {sizeRule=Fill, value=0.75}}}
  }
  Border_154 {
   widgetClassPath=&Border
   parent=VerticalBox_38
   widget=Border_154
   properties {
    background {
     bIsDynamicallyLoaded=false
     drawAs=Box
     imageType=NoImage
     mirroring=NoMirror
     resourceName=None
     resourceObject=null
     tiling=NoTile
     imageSize=<32 32>
     margin=<ltrb 0 0 0 0>
     outlineSettings {
      bUseBrushTransparency=false
      roundingType=HalfHeightRadius
      width=0
      color {colorUseRule=UseColor_Specified, specifiedColor=<rgba 0 0 0 0>}
      cornerRadii=<0 0 0 0>
     }
     tintColor {colorUseRule=UseColor_Specified, specifiedColor=<rgba 1 1 1 1>}
     uVRegion {bIsValid=false, max=<0 0>, min=<0 0>}
    }
    brushColor=<rgba 0 0 0 0.2>
    padding=<ltrb 5 5 5 5>
   }
   slot {class=VerticalBoxSlot, properties {padding=<ltrb 2.5 2.5 2.5 2.5>, size {sizeRule=Fill, value=0.8}}}
  }
  ScrollBox_61 {
   widgetClassPath=&ScrollBox
   parent=Border_154
   widget=ScrollBox_61
   properties {
    alwaysShowScrollbar=true
    alwaysShowScrollbarTrack=true
    consumeMouseWheel=Always
    scrollbarThickness=<8 8>
    widgetBarStyle {
     thickness=4
     draggedThumbImage {
      bIsDynamicallyLoaded=false
      drawAs=RoundedBox
      imageType=NoImage
      mirroring=NoMirror
      resourceName=None
      resourceObject=null
      tiling=NoTile
      imageSize=<0 0>
      margin=<ltrb 0 0 0 0>
      outlineSettings {
       bUseBrushTransparency=false
       roundingType=FixedRadius
       width=0
       color {colorUseRule=UseColor_Specified, specifiedColor=<rgba 0 0 0 0>}
       cornerRadii=<4 4 4 4>
      }
      tintColor {colorUseRule=UseColor_Specified, specifiedColor=<rgba 0.2159 0.2159 0.2159 1>}
      uVRegion {bIsValid=false, max=<0 0>, min=<0 0>}
     }
     horizontalBackgroundImage {
      bIsDynamicallyLoaded=false
      drawAs=NoDrawType
      imageType=NoImage
      mirroring=NoMirror
      resourceName=None
      resourceObject=null
      tiling=NoTile
      imageSize=<0 0>
      margin=<ltrb 0 0 0 0>
      outlineSettings {
       bUseBrushTransparency=false
       roundingType=HalfHeightRadius
       width=0
       color {colorUseRule=UseColor_Specified, specifiedColor=<rgba 0 0 0 0>}
       cornerRadii=<0 0 0 0>
      }
      tintColor {colorUseRule=UseColor_Specified, specifiedColor=<rgba 1 1 1 1>}
      uVRegion {bIsValid=false, max=<0 0>, min=<0 0>}
     }
     horizontalBottomSlotImage {
      bIsDynamicallyLoaded=false
      drawAs=NoDrawType
      imageType=NoImage
      mirroring=NoMirror
      resourceName=None
      resourceObject=null
      tiling=NoTile
      imageSize=<0 0>
      margin=<ltrb 0 0 0 0>
      outlineSettings {
       bUseBrushTransparency=false
       roundingType=HalfHeightRadius
       width=0
       color {colorUseRule=UseColor_Specified, specifiedColor=<rgba 0 0 0 0>}
       cornerRadii=<0 0 0 0>
      }
      tintColor {colorUseRule=UseColor_Specified, specifiedColor=<rgba 1 1 1 1>}
      uVRegion {bIsValid=false, max=<0 0>, min=<0 0>}
     }
     horizontalTopSlotImage {
      bIsDynamicallyLoaded=false
      drawAs=NoDrawType
      imageType=NoImage
      mirroring=NoMirror
      resourceName=None
      resourceObject=null
      tiling=NoTile
      imageSize=<0 0>
      margin=<ltrb 0 0 0 0>
      outlineSettings {
       bUseBrushTransparency=false
       roundingType=HalfHeightRadius
       width=0
       color {colorUseRule=UseColor_Specified, specifiedColor=<rgba 0 0 0 0>}
       cornerRadii=<0 0 0 0>
      }
      tintColor {colorUseRule=UseColor_Specified, specifiedColor=<rgba 1 1 1 1>}
      uVRegion {bIsValid=false, max=<0 0>, min=<0 0>}
     }
     hoveredThumbImage {
      bIsDynamicallyLoaded=false
      drawAs=RoundedBox
      imageType=NoImage
      mirroring=NoMirror
      resourceName=None
      resourceObject=null
      tiling=NoTile
      imageSize=<0 0>
      margin=<ltrb 0 0 0 0>
      outlineSettings {
       bUseBrushTransparency=false
       roundingType=FixedRadius
       width=0
       color {colorUseRule=UseColor_Specified, specifiedColor=<rgba 0 0 0 0>}
       cornerRadii=<4 4 4 4>
      }
      tintColor {colorUseRule=UseColor_Specified, specifiedColor=<rgba 0.2159 0.2159 0.2159 1>}
      uVRegion {bIsValid=false, max=<0 0>, min=<0 0>}
     }
     normalThumbImage {
      bIsDynamicallyLoaded=false
      drawAs=RoundedBox
      imageType=NoImage
      mirroring=NoMirror
      resourceName=None
      resourceObject=null
      tiling=NoTile
      imageSize=<0 0>
      margin=<ltrb 0 0 0 0>
      outlineSettings {
       bUseBrushTransparency=false
       roundingType=FixedRadius
       width=0
       color {colorUseRule=UseColor_Specified, specifiedColor=<rgba 0 0 0 0>}
       cornerRadii=<4 4 4 4>
      }
      tintColor {colorUseRule=UseColor_Specified, specifiedColor=<rgba 0.0953 0.0953 0.0953 1>}
      uVRegion {bIsValid=false, max=<0 0>, min=<0 0>}
     }
     verticalBackgroundImage {
      bIsDynamicallyLoaded=false
      drawAs=NoDrawType
      imageType=NoImage
      mirroring=NoMirror
      resourceName=None
      resourceObject=null
      tiling=NoTile
      imageSize=<0 0>
      margin=<ltrb 0 0 0 0>
      outlineSettings {
       bUseBrushTransparency=false
       roundingType=HalfHeightRadius
       width=0
       color {colorUseRule=UseColor_Specified, specifiedColor=<rgba 0 0 0 0>}
       cornerRadii=<0 0 0 0>
      }
      tintColor {colorUseRule=UseColor_Specified, specifiedColor=<rgba 1 1 1 1>}
      uVRegion {bIsValid=false, max=<0 0>, min=<0 0>}
     }
     verticalBottomSlotImage {
      bIsDynamicallyLoaded=false
      drawAs=NoDrawType
      imageType=NoImage
      mirroring=NoMirror
      resourceName=None
      resourceObject=null
      tiling=NoTile
      imageSize=<0 0>
      margin=<ltrb 0 0 0 0>
      outlineSettings {
       bUseBrushTransparency=false
       roundingType=HalfHeightRadius
       width=0
       color {colorUseRule=UseColor_Specified, specifiedColor=<rgba 0 0 0 0>}
       cornerRadii=<0 0 0 0>
      }
      tintColor {colorUseRule=UseColor_Specified, specifiedColor=<rgba 1 1 1 1>}
      uVRegion {bIsValid=false, max=<0 0>, min=<0 0>}
     }
     verticalTopSlotImage {
      bIsDynamicallyLoaded=false
      drawAs=NoDrawType
      imageType=NoImage
      mirroring=NoMirror
      resourceName=None
      resourceObject=null
      tiling=NoTile
      imageSize=<0 0>
      margin=<ltrb 0 0 0 0>
      outlineSettings {
       bUseBrushTransparency=false
       roundingType=HalfHeightRadius
       width=0
       color {colorUseRule=UseColor_Specified, specifiedColor=<rgba 0 0 0 0>}
       cornerRadii=<0 0 0 0>
      }
      tintColor {colorUseRule=UseColor_Specified, specifiedColor=<rgba 1 1 1 1>}
      uVRegion {bIsValid=false, max=<0 0>, min=<0 0>}
     }
    }
    widgetStyle {
     barThickness=8
     bottomShadowBrush {
      bIsDynamicallyLoaded=false
      drawAs=Box
      imageType=FullColor
      mirroring=NoMirror
      resourceName=../../../Engine/Content/Slate/Common/ScrollBoxShadowBottom.png
      resourceObject=null
      tiling=NoTile
      imageSize=<16 8>
      margin=<ltrb 0.5 0 0.5 1>
      outlineSettings {
       bUseBrushTransparency=false
       roundingType=HalfHeightRadius
       width=0
       color {colorUseRule=UseColor_Specified, specifiedColor=<rgba 0 0 0 0>}
       cornerRadii=<0 0 0 0>
      }
      tintColor {colorUseRule=UseColor_Specified, specifiedColor=<rgba 1 1 1 1>}
      uVRegion {bIsValid=false, max=<0 0>, min=<0 0>}
     }
     horizontalScrolledContentPadding=<ltrb 0 0 1 0>
     leftShadowBrush {
      bIsDynamicallyLoaded=false
      drawAs=Box
      imageType=FullColor
      mirroring=NoMirror
      resourceName=../../../Engine/Content/Slate/Common/ScrollBoxShadowLeft.png
      resourceObject=null
      tiling=NoTile
      imageSize=<8 16>
      margin=<ltrb 1 0.5 0 0.5>
      outlineSettings {
       bUseBrushTransparency=false
       roundingType=HalfHeightRadius
       width=0
       color {colorUseRule=UseColor_Specified, specifiedColor=<rgba 0 0 0 0>}
       cornerRadii=<0 0 0 0>
      }
      tintColor {colorUseRule=UseColor_Specified, specifiedColor=<rgba 1 1 1 1>}
      uVRegion {bIsValid=false, max=<0 0>, min=<0 0>}
     }
     rightShadowBrush {
      bIsDynamicallyLoaded=false
      drawAs=Box
      imageType=FullColor
      mirroring=NoMirror
      resourceName=../../../Engine/Content/Slate/Common/ScrollBoxShadowRight.png
      resourceObject=null
      tiling=NoTile
      imageSize=<8 16>
      margin=<ltrb 0 0.5 1 0.5>
      outlineSettings {
       bUseBrushTransparency=false
       roundingType=HalfHeightRadius
       width=0
       color {colorUseRule=UseColor_Specified, specifiedColor=<rgba 0 0 0 0>}
       cornerRadii=<0 0 0 0>
      }
      tintColor {colorUseRule=UseColor_Specified, specifiedColor=<rgba 1 1 1 1>}
      uVRegion {bIsValid=false, max=<0 0>, min=<0 0>}
     }
     topShadowBrush {
      bIsDynamicallyLoaded=false
      drawAs=Box
      imageType=FullColor
      mirroring=NoMirror
      resourceName=../../../Engine/Content/Slate/Common/ScrollBoxShadowTop.png
      resourceObject=null
      tiling=NoTile
      imageSize=<16 8>
      margin=<ltrb 0.5 1 0.5 0>
      outlineSettings {
       bUseBrushTransparency=false
       roundingType=HalfHeightRadius
       width=0
       color {colorUseRule=UseColor_Specified, specifiedColor=<rgba 0 0 0 0>}
       cornerRadii=<0 0 0 0>
      }
      tintColor {colorUseRule=UseColor_Specified, specifiedColor=<rgba 1 1 1 1>}
      uVRegion {bIsValid=false, max=<0 0>, min=<0 0>}
     }
     verticalScrolledContentPadding=<ltrb 0 0 0 1>
    }
   }
   slot {class=BorderSlot, properties {padding=<ltrb 5 5 5 5>}}
  }
  SoundCueGraphTree {
   widgetClassPath=/Script/UMG.TreeView
   bIsVariable=true
   parent=ScrollBox_61
   widget=SoundCueGraphTree
   properties {
    bEnableScrollAnimation=true
    consumeMouseWheel=Never
    entryWidgetClass=&/EUW_SoundCueEditor_Item.EUW_SoundCueEditor_Item_C
    scrollBarStyle {
     thickness=8
     draggedThumbImage {
      bIsDynamicallyLoaded=false
      drawAs=RoundedBox
      imageType=NoImage
      mirroring=NoMirror
      resourceName=None
      resourceObject=null
      tiling=NoTile
      imageSize=<0 0>
      margin=<ltrb 0 0 0 0>
      outlineSettings {
       bUseBrushTransparency=false
       roundingType=FixedRadius
       width=0
       color {colorUseRule=UseColor_Specified, specifiedColor=<rgba 0 0 0 0>}
       cornerRadii=<4 4 4 4>
      }
      tintColor {colorUseRule=UseColor_Specified, specifiedColor=<rgba 0.2159 0.2159 0.2159 1>}
      uVRegion {bIsValid=false, max=<0 0>, min=<0 0>}
     }
     horizontalBackgroundImage {
      bIsDynamicallyLoaded=false
      drawAs=NoDrawType
      imageType=NoImage
      mirroring=NoMirror
      resourceName=None
      resourceObject=null
      tiling=NoTile
      imageSize=<0 0>
      margin=<ltrb 0 0 0 0>
      outlineSettings {
       bUseBrushTransparency=false
       roundingType=HalfHeightRadius
       width=0
       color {colorUseRule=UseColor_Specified, specifiedColor=<rgba 0 0 0 0>}
       cornerRadii=<0 0 0 0>
      }
      tintColor {colorUseRule=UseColor_Specified, specifiedColor=<rgba 1 1 1 1>}
      uVRegion {bIsValid=false, max=<0 0>, min=<0 0>}
     }
     horizontalBottomSlotImage {
      bIsDynamicallyLoaded=false
      drawAs=NoDrawType
      imageType=NoImage
      mirroring=NoMirror
      resourceName=None
      resourceObject=null
      tiling=NoTile
      imageSize=<0 0>
      margin=<ltrb 0 0 0 0>
      outlineSettings {
       bUseBrushTransparency=false
       roundingType=HalfHeightRadius
       width=0
       color {colorUseRule=UseColor_Specified, specifiedColor=<rgba 0 0 0 0>}
       cornerRadii=<0 0 0 0>
      }
      tintColor {colorUseRule=UseColor_Specified, specifiedColor=<rgba 1 1 1 1>}
      uVRegion {bIsValid=false, max=<0 0>, min=<0 0>}
     }
     horizontalTopSlotImage {
      bIsDynamicallyLoaded=false
      drawAs=NoDrawType
      imageType=NoImage
      mirroring=NoMirror
      resourceName=None
      resourceObject=null
      tiling=NoTile
      imageSize=<0 0>
      margin=<ltrb 0 0 0 0>
      outlineSettings {
       bUseBrushTransparency=false
       roundingType=HalfHeightRadius
       width=0
       color {colorUseRule=UseColor_Specified, specifiedColor=<rgba 0 0 0 0>}
       cornerRadii=<0 0 0 0>
      }
      tintColor {colorUseRule=UseColor_Specified, specifiedColor=<rgba 1 1 1 1>}
      uVRegion {bIsValid=false, max=<0 0>, min=<0 0>}
     }
     hoveredThumbImage {
      bIsDynamicallyLoaded=false
      drawAs=RoundedBox
      imageType=NoImage
      mirroring=NoMirror
      resourceName=None
      resourceObject=null
      tiling=NoTile
      imageSize=<0 0>
      margin=<ltrb 0 0 0 0>
      outlineSettings {
       bUseBrushTransparency=false
       roundingType=FixedRadius
       width=0
       color {colorUseRule=UseColor_Specified, specifiedColor=<rgba 0 0 0 0>}
       cornerRadii=<4 4 4 4>
      }
      tintColor {colorUseRule=UseColor_Specified, specifiedColor=<rgba 0.2159 0.2159 0.2159 1>}
      uVRegion {bIsValid=false, max=<0 0>, min=<0 0>}
     }
     normalThumbImage {
      bIsDynamicallyLoaded=false
      drawAs=RoundedBox
      imageType=NoImage
      mirroring=NoMirror
      resourceName=None
      resourceObject=null
      tiling=NoTile
      imageSize=<0 0>
      margin=<ltrb 0 0 0 0>
      outlineSettings {
       bUseBrushTransparency=false
       roundingType=FixedRadius
       width=0
       color {colorUseRule=UseColor_Specified, specifiedColor=<rgba 0 0 0 0>}
       cornerRadii=<4 4 4 4>
      }
      tintColor {colorUseRule=UseColor_Specified, specifiedColor=<rgba 0.0953 0.0953 0.0953 1>}
      uVRegion {bIsValid=false, max=<0 0>, min=<0 0>}
     }
     verticalBackgroundImage {
      bIsDynamicallyLoaded=false
      drawAs=NoDrawType
      imageType=NoImage
      mirroring=NoMirror
      resourceName=None
      resourceObject=null
      tiling=NoTile
      imageSize=<0 0>
      margin=<ltrb 0 0 0 0>
      outlineSettings {
       bUseBrushTransparency=false
       roundingType=HalfHeightRadius
       width=0
       color {colorUseRule=UseColor_Specified, specifiedColor=<rgba 0 0 0 0>}
       cornerRadii=<0 0 0 0>
      }
      tintColor {colorUseRule=UseColor_Specified, specifiedColor=<rgba 1 1 1 1>}
      uVRegion {bIsValid=false, max=<0 0>, min=<0 0>}
     }
     verticalBottomSlotImage {
      bIsDynamicallyLoaded=false
      drawAs=NoDrawType
      imageType=NoImage
      mirroring=NoMirror
      resourceName=None
      resourceObject=null
      tiling=NoTile
      imageSize=<0 0>
      margin=<ltrb 0 0 0 0>
      outlineSettings {
       bUseBrushTransparency=false
       roundingType=HalfHeightRadius
       width=0
       color {colorUseRule=UseColor_Specified, specifiedColor=<rgba 0 0 0 0>}
       cornerRadii=<0 0 0 0>
      }
      tintColor {colorUseRule=UseColor_Specified, specifiedColor=<rgba 1 1 1 1>}
      uVRegion {bIsValid=false, max=<0 0>, min=<0 0>}
     }
     verticalTopSlotImage {
      bIsDynamicallyLoaded=false
      drawAs=NoDrawType
      imageType=NoImage
      mirroring=NoMirror
      resourceName=None
      resourceObject=null
      tiling=NoTile
      imageSize=<0 0>
      margin=<ltrb 0 0 0 0>
      outlineSettings {
       bUseBrushTransparency=false
       roundingType=HalfHeightRadius
       width=0
       color {colorUseRule=UseColor_Specified, specifiedColor=<rgba 0 0 0 0>}
       cornerRadii=<0 0 0 0>
      }
      tintColor {colorUseRule=UseColor_Specified, specifiedColor=<rgba 1 1 1 1>}
      uVRegion {bIsValid=false, max=<0 0>, min=<0 0>}
     }
    }
   }
   slot {class=ScrollBoxSlot}
  }
  VerticalBox {
   widgetClassPath=&VerticalBox
   parent=HorizontalBox_180
   widget=VerticalBox
   slot {class=HorizontalBoxSlot, properties {size {sizeRule=Fill, value=0.75}}}
  }
  Border_1 {
   widgetClassPath=&Border
   parent=VerticalBox
   widget=Border_1
   properties {
    verticalAlignment=VAlign_Top
    background {
     bIsDynamicallyLoaded=false
     drawAs=Box
     imageType=NoImage
     mirroring=NoMirror
     resourceName=None
     resourceObject=null
     tiling=NoTile
     imageSize=<32 32>
     margin=<ltrb 0 0 0 0>
     outlineSettings {
      bUseBrushTransparency=false
      roundingType=HalfHeightRadius
      width=0
      color {colorUseRule=UseColor_Specified, specifiedColor=<rgba 0 0 0 0>}
      cornerRadii=<0 0 0 0>
     }
     tintColor {colorUseRule=UseColor_Specified, specifiedColor=<rgba 1 1 1 1>}
     uVRegion {bIsValid=false, max=<0 0>, min=<0 0>}
    }
    brushColor=<rgba 0 0 0 0.2>
    padding=<ltrb 5 5 5 5>
   }
   slot {class=VerticalBoxSlot, properties {padding=<ltrb 2.5 2.5 2.5 2.5>, size {sizeRule=Fill, value=1}}}
  }
  SoundCuePicker {
   widgetClassPath=&SinglePropertyView
   parent=Border_1
   widget=SoundCuePicker
   properties {propertyName=SelectedCue}
   slot {class=BorderSlot, properties {verticalAlignment=VAlign_Top, padding=<ltrb 5 5 5 5>}}
  }
 ]
}
